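# Looping spatial ambience cue of bat wing flapping for environment bat actors.
# A looping flutter wave runs through a Modulator and two nested, endlessly looping Envelopers: a slower volume swell inside a shorter volume and pitch drift, varying the single loop over time.
# Overridden attenuation adds distance low-pass and listener focus, concurrency allows one voice per owner, and sends feed an exterior ambient bus and reverb.
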SoundCue /Game/Sounds/Asteria/Environment_Actors/Bats/Spot_Bats_Wing_Flap_Loop_Cue
root=Enveloper_0
nodes [
 {
  id=Enveloper_0
  children [Enveloper_1]
  properties {
   bLoop=true
   bLoopIndefinitely=true
   loopEnd=10
   pitchCurve {editorCurveData {keys [
      {interpMode=RCIM_Cubic, time=0}
      {arriveTangent=-0.0147, interpMode=RCIM_Cubic, leaveTangent=-0.0147, time=3.1287, value=1.2043}
      {arriveTangent=0.0004, interpMode=RCIM_Cubic, leaveTangent=0.0004, tangentMode=RCTM_User, time=6.7844, value=0.9}
      {time=10}
     ]}}
   volumeCurve {editorCurveData {keys [<at 0 1 RCIM_Cubic>,<at 5 0.8 RCIM_Cubic>,<at 10 1>]}}
  }
 }
 {
  id=Enveloper_1
  children [Modulator_1]
  properties {
   bLoop=true
   bLoopIndefinitely=true
   loopEnd=20
   pitchCurve {editorCurveData {keys [{time=0}]}}
   volumeCurve {editorCurveData {keys [<at 0 1 RCIM_Cubic>,<at 10 0.5 RCIM_Cubic>,<at 20 1>]}}
  }
 }
 {id=Modulator_1, children [WavePlayer_18]}
 {id=WavePlayer_18, properties {bLooping=true, soundWaveAssetPtr=&/Spot_Bat_Flutter_Loop}}
]
properties {
 bOverrideAttenuation=true
 bOverrideConcurrency=true
 maxDistance=3500
 soundClassObject=&Fort_Ambient_Indoors_Athena
 volumeMultiplier=0.752
 attenuationOverrides {
  bAttenuateWithLPF=true
  bEnableListenerFocus=true
  falloffDistance=3000
  lPFFrequencyAtMax=4000
  lPFRadiusMax=4000
  lPFRadiusMin=750
  nonFocusVolumeAttenuation=0.5
  stereoSpread=1000
  attenuationShapeExtents {x=500}
 }
 busSends [{audioBus=null, minSendLevel=0.6, sendLevel=1, soundSourceBus=&AmbientAudio_Exterior_SourceBus}]
 concurrencyOverrides {bLimitToOwner=true, maxCount=1, voiceStealReleaseTime=3}
 soundSubmixSends [{maxSendLevel=1.2, minSendLevel=1, soundSubmix=&Fortnite_Reverb_Default}]
}
